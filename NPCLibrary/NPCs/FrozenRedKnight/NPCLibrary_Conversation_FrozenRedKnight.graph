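# Conversation database for the Frozen Red Knight NPC, defining its greeting and the branches a player can take from it.
# The entry tag leads to one speech node with a default line plus contextual lines for a first-ever meeting and for players wearing specific outfits.
# Requirements gate each contextual line and priority curve rows rank them.
# The speech then links out to the shared NPC Library quest and basic services conversations.
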ConversationDatabase /NPCLibrary/NPCs/FrozenRedKnight/NPCLibrary_Conversation_FrozenRedKnight
entry_points [{tag=Conversation.Entry.FrozenRedKnight, destinations [#1]}]
nodes [
 ConversationEntryPointNode_7 {
  id=#1
  class=ConversationEntryPointNode
  outputs [#2]
  properties {entryTag {tagName=Conversation.Entry.FrozenRedKnight}}
 }
 Speech_C_4 {
  id=#2
  class=Speech_C
  outputs [#3, #4]
  properties {
   nodeName=""
   generalConfig {
    defaultMessage=Let's cause some chaos.
    contextualMessages [
     {
      message=Have you seen frozen kingdoms burn? It's quite funny to watch.
      contextRequirements [{requirement {class=FortControllerRequirement_IsFirstEverConversationWithNPC}}]
      priority {curve {rowName=Tandem.Speech.Priority.FirstEverInteraction}}
     }
     {
      message="Sorry, kid, but I'm rooting for the vampires."
      contextRequirements [{requirement {class=FortControllerRequirement_HasCID, properties {allowedCIds [&Character_SharpFang]}}}]
      priority {curve {rowName=Tandem.Speech.Priority.OutfitReaction}}
     }
     {
      message="With your power, we could ruin this place. Interested?"
      contextRequirements [{requirement {class=FortControllerRequirement_HasCID, properties {allowedCIds [/BRCosmetics/Athena/Items/Cosmetics/Characters/Character_Apprentice]}}}]
      priority {curve {rowName=Tandem.Speech.Priority.OutfitReaction}}
     }
     {
      message="You serve a cold, barren rock. Some hero."
      contextRequirements [{requirement {class=FortControllerRequirement_HasCID, properties {allowedCIds [&Character_Sunlit]}}}]
      priority {curve {rowName=Tandem.Speech.Priority.OutfitReaction}}
     }
     {
      message=Zombies filled your world with chaos? You should thank them.
      contextRequirements [{requirement {class=FortControllerRequirement_HasCID, properties {allowedCIds [/BRCosmetics/Athena/Items/Cosmetics/Characters/Character_Chainmail]}}}]
      priority {curve {rowName=Tandem.Speech.Priority.OutfitReaction}}
     }
     {
      message=I'd rather see your kingdom burn than in Snowheart's icy grip.
      contextRequirements [{requirement {class=FortControllerRequirement_HasCID, properties {allowedCIds [&Character_Citadel]}}}]
      priority {curve {rowName=Tandem.Speech.Priority.OutfitReaction}}
     }
     {
      message="Why, if it isn't the ice cold warrior of the wastes."
      contextRequirements [{requirement {
         class=FortControllerRequirement_HasCID
         properties {allowedCIds [/BRCosmetics/Athena/Items/Cosmetics/Characters/CID_656_Athena_Commando_M_TeriyakiFishFreezerBurn]}
        }}]
      priority {curve {rowName=Tandem.Speech.Priority.OutfitReaction}}
     }
     {
      message=Will your icy heart protect you from the heat of my rage?
      contextRequirements [{requirement {
         class=FortControllerRequirement_HasCID
         properties {allowedCIds [/BRCosmetics/Athena/Items/Cosmetics/Characters/CID_983_Athena_Commando_F_CupidWinter]}
        }}]
      priority {curve {rowName=Tandem.Speech.Priority.OutfitReaction}}
     }
    ]
   }
  }
 }
 ConversationLinkNode_0 {id=#4, class=ConversationLinkNode, properties {remoteEntryTag {tagName=Conversation.Entry.BasicServices}}}
 ConversationLinkNode_1 {id=#3, class=ConversationLinkNode, properties {remoteEntryTag {tagName=Conversation.Entry.NPCLibrary.Quests}}}
]
